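# Sound cue for snow drift movement in Athena, playing outdoor ambient one-shot or looping wave variations with its own attenuation.
# A quality level node routes the first two quality slots to a random pick of nine waves and the last two to a pick among only the first four.
# Both random nodes weight their waves equally and are excluded from branch culling.
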SoundCue /Game/Athena/Sounds/Ambient/SnowDrift/SnowDrift_Movement_Cue
root=QualityLevel_0
nodes [
 {id=QualityLevel_0, children [Random_0,Random_0,Random_1,Random_1]}
 {
  id=Random_0
  children [WavePlayer_0,WavePlayer_1,WavePlayer_2,WavePlayer_3,WavePlayer_4,WavePlayer_5,WavePlayer_6,WavePlayer_7,WavePlayer_8]
  properties {bShouldExcludeFromBranchCulling=true, weights [1,1,1,1,1,1,1,1,1]}
 }
 {
  id=Random_1
  children [WavePlayer_9, WavePlayer_10, WavePlayer_11, WavePlayer_12]
  properties {bShouldExcludeFromBranchCulling=true, weights [1,1,1,1]}
 }
 {id=WavePlayer_0, properties {soundWaveAssetPtr=&/SnowDrift_Movement_01}}
 {id=WavePlayer_1, properties {soundWaveAssetPtr=&/SnowDrift_Movement_02}}
 {id=WavePlayer_10, properties {soundWaveAssetPtr=&SnowDrift_Movement_02}}
 {id=WavePlayer_11, properties {soundWaveAssetPtr=&/SnowDrift_Movement_03}}
 {id=WavePlayer_12, properties {soundWaveAssetPtr=&/SnowDrift_Movement_04}}
 {id=WavePlayer_2, properties {soundWaveAssetPtr=&SnowDrift_Movement_03}}
 {id=WavePlayer_3, properties {soundWaveAssetPtr=&SnowDrift_Movement_04}}
 {id=WavePlayer_4, properties {soundWaveAssetPtr=&/SnowDrift_Movement_05}}
 {id=WavePlayer_5, properties {soundWaveAssetPtr=&/SnowDrift_Movement_06}}
 {id=WavePlayer_6, properties {soundWaveAssetPtr=&/SnowDrift_Movement_07}}
 {id=WavePlayer_7, properties {soundWaveAssetPtr=&/SnowDrift_Movement_08}}
 {id=WavePlayer_8, properties {soundWaveAssetPtr=&/SnowDrift_Movement_09}}
 {id=WavePlayer_9, properties {soundWaveAssetPtr=&SnowDrift_Movement_01}}
]
properties {
 attenuationSettings=&/SnowDrift_Movement_Att
 maxDistance=10600
 soundClassObject=&Fort_Ambient_Outdoors_Athena
 volumeMultiplier=2.506
}
